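# Looping engine burn sound cue for the BarnTowel Armory's thrust, routed to the vehicle engine sound class.
# A distance crossfade blends a close thrust loop into a distant one over an overlapping range, with the distant layer fading out near the cue's max distance.
# Each layer has its own attenuation asset, the distant one borrowing SkyRift music settings, plus a short envelope fade-in; a fixed modulator lowers the close layer's volume.
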SoundCue /BarnTowelArmoryGameplay/Assets/Sounds/SC_BarnTowelArmory_Armory_EngineBurn_Loop
root=DistanceCrossFade_0
nodes [
 {id=Attenuation_0, children [WavePlayer_0], properties {attenuationSettings=&/AS_BarnTowel_ShipLanding}}
 {id=Attenuation_1, children [WavePlayer_1], properties {attenuationSettings=&/AS_BarnTowelSkyRift_Music}}
 {
  id=DistanceCrossFade_0
  children [Modulator_0,Enveloper_1]
  properties {crossFadeInput [
    {fadeInDistanceEnd=0, fadeInDistanceStart=0, fadeOutDistanceEnd=35000, fadeOutDistanceStart=15000, volume=0.8}
    {fadeInDistanceEnd=29000, fadeInDistanceStart=10000, fadeOutDistanceEnd=100000, fadeOutDistanceStart=80000, volume=1}
   ]}
 }
 {
  id=Enveloper_0
  children [Attenuation_0]
  properties {pitchCurve {editorCurveData {keys [{time=0}]}}, volumeCurve {editorCurveData {keys [<at 0 0>,<at 0.4 1>]}}}
 }
 {
  id=Enveloper_1
  children [Attenuation_1]
  properties {pitchCurve {editorCurveData {keys [{time=0}]}}, volumeCurve {editorCurveData {keys [<at 0 0>,<at 0.5 1>]}}}
 }
 {id=Modulator_0, children [Enveloper_0], properties {volumeMax=0.7, volumeMin=0.7}}
 {
  id=WavePlayer_0
  properties {bLooping=true, soundWaveAssetPtr=/BarnTowelArmoryGameplay/Assets/Sounds/wav/BarnTowelArmory_Armory_Thrust_lp_Close_01}
 }
 {
  id=WavePlayer_1
  properties {bLooping=true, soundWaveAssetPtr=/BarnTowelArmoryGameplay/Assets/Sounds/wav/BarnTowelArmory_Armory_Thrust_lp_Dist_01}
 }
]
properties {maxDistance=100000, soundClassObject=&Fort_Vehicles_Engine, volumeMultiplier=1.4}
