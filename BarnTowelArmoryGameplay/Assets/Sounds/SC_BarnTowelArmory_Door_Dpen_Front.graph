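# Sound cue for the front door opening on the BarnTowel Armory.
# A fixed half-second delay precedes a single door-open wave, played at reduced volume in the gadgets sound class.
# Spatialization uses the shared ship-landing attenuation asset; the cue's inline attenuation values are present without the override enabled.
# An inline concurrency override caps simultaneous instances at two.
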SoundCue /BarnTowelArmoryGameplay/Assets/Sounds/SC_BarnTowelArmory_Door_Dpen_Front
root=Delay_0
nodes [
 {id=Delay_0, children [WavePlayer_0], properties {delayMax=0.5, delayMin=0.5}}
 {id=WavePlayer_0, properties {soundWaveAssetPtr=/BarnTowelArmoryGameplay/Assets/Sounds/wav/BarnTowelArmory_Armory_DoorOpen_Front}}
]
properties {
 attenuationSettings=&/AS_BarnTowel_ShipLanding
 bOverrideConcurrency=true
 maxDistance=53000
 soundClassObject=&Fort_Gadgets
 volumeMultiplier=0.55
 attenuationOverrides {
  bAttenuateWithLPF=true
  bEnableOcclusion=true
  falloffDistance=20000
  hPFFrequencyAtMin=95
  lPFFrequencyAtMax=900
  lPFRadiusMax=3000
  lPFRadiusMin=100
  occlusionLowPassFilterFrequency=950
  occlusionTraceChannel=ECC_GameTraceChannel6
  occlusionVolumeAttenuation=0.8
  attenuationShapeExtents {x=500}
 }
 concurrencyOverrides {maxCount=2}
}
