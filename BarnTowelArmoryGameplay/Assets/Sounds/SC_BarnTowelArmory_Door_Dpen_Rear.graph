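# Sound cue that plays the rear door opening sound for the BarnTowel Armory.
# A single wave player feeds the output directly, with reduced volume and the Fort_Gadgets sound class.
# Spatialization comes from the shared ship-landing attenuation asset, as the cue's own attenuation values are stored without the override enabled.
# A concurrency override on the cue caps simultaneous instances.
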SoundCue /BarnTowelArmoryGameplay/Assets/Sounds/SC_BarnTowelArmory_Door_Dpen_Rear
root=WavePlayer_2
nodes [{id=WavePlayer_2, properties {soundWaveAssetPtr=/BarnTowelArmoryGameplay/Assets/Sounds/wav/BarnTowelArmory_Armory_DoorOpen_Rear}}]
properties {
 attenuationSettings=&/AS_BarnTowel_ShipLanding
 bOverrideConcurrency=true
 maxDistance=53000
 soundClassObject=&Fort_Gadgets
 volumeMultiplier=0.6
 attenuationOverrides {
  bAttenuateWithLPF=true
  bEnableOcclusion=true
  falloffDistance=20000
  hPFFrequencyAtMin=95
  lPFFrequencyAtMax=900
  lPFRadiusMax=3000
  lPFRadiusMin=100
  occlusionLowPassFilterFrequency=950
  occlusionTraceChannel=ECC_GameTraceChannel6
  occlusionVolumeAttenuation=0.8
  attenuationShapeExtents {x=500}
 }
 concurrencyOverrides {maxCount=5}
}
