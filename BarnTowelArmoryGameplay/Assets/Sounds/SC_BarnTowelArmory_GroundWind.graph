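# Sound cue for the looping ground wind of the BarnTowel Armory, classed as a gadget sound.
# A looping wave player feeds a default Modulator, then an Enveloper whose volume curve swells in, dips, and climbs back to full over a few seconds.
# The cue uses wind attenuation, sends to the default reverb submix, and overrides concurrency to allow one instance with a retrigger time.
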SoundCue /BarnTowelArmoryGameplay/Assets/Sounds/SC_BarnTowelArmory_GroundWind
root=Enveloper_1
nodes [
 {
  id=Enveloper_1
  children [Modulator_0]
  properties {pitchCurve {editorCurveData {keys [{time=0}]}}, volumeCurve {editorCurveData {keys [<at 0 0>,<at 1 1>,<at 1.4232 0.6521>,<at 4.7101 1>]}}}
 }
 {id=Modulator_0, children [WavePlayer_0]}
 {
  id=WavePlayer_0
  properties {bLooping=true, soundWaveAssetPtr=/BarnTowelArmoryGameplay/Assets/Sounds/wav/BarnTowelArmory_Armory_GroundWind_lp}
 }
]
properties {
 attenuationSettings=&/AS_BarnTowel_Wind
 bOverrideConcurrency=true
 maxDistance=5500
 soundClassObject=&Fort_Gadgets
 volumeMultiplier=0.45
 concurrencyOverrides {maxCount=1, retriggerTime=2}
 soundSubmixSends [{minSendLevel=0.3, soundSubmix=&Fortnite_Reverb_Default}]
}
