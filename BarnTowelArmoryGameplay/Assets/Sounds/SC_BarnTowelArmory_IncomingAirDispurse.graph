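# Sound cue for the in-flight air disbursement noise of the incoming BarnTowel Armory, played as an in-game SFX.
# A single wave player feeds an Enveloper whose volume curve starts boosted, fades almost to silence over several seconds, then swells back up sharply at the end.
# The cue raises overall volume, uses the shared ship-landing attenuation, and overrides concurrency to cap simultaneous instances and retriggers.
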SoundCue /BarnTowelArmoryGameplay/Assets/Sounds/SC_BarnTowelArmory_IncomingAirDispurse
root=Enveloper_1
nodes [
 {
  id=Enveloper_1
  children [WavePlayer_0]
  properties {
   pitchCurve {editorCurveData {keys [{time=0}]}}
   volumeCurve {editorCurveData {keys [
      {arriveTangent=0.1456, interpMode=RCIM_Cubic, leaveTangent=-0.0274, tangentMode=RCTM_User, time=0, value=2}
      <at 0.002 2>
      <at 1.3519 1.9008>
      {arriveTangent=-0.5935, interpMode=RCIM_Cubic, leaveTangent=-0.5935, time=2.7593, value=1.1124}
      {arriveTangent=0.039, interpMode=RCIM_Cubic, leaveTangent=0.039, time=4.407, value=0.0875}
      <at 4.8961 1.1958>
     ]}}
  }
 }
 {id=WavePlayer_0, properties {soundWaveAssetPtr=/BarnTowelArmoryGameplay/Assets/Sounds/wav/BarnTowelArmory_Armory_AirDisbursmentInFlight}}
]
properties {
 attenuationSettings=&/AS_BarnTowel_ShipLanding
 bOverrideConcurrency=true
 maxDistance=53000
 soundClassObject=&Fort_SFX_InGame
 volumeMultiplier=1.75
 concurrencyOverrides {maxCount=2, retriggerTime=2}
}
