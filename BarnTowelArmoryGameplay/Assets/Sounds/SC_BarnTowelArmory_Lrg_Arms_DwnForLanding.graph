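# Sound cue for the large BarnTowel Armory's arms lowering for landing, built from two timed wave parts.
# A mixer layers part one after a short fixed delay and part two a few seconds later at a higher input volume, forming a staged mechanical sequence.
# Playback uses the ship-landing attenuation with a very long range, the gadgets sound class, slightly lowered pitch and volume, and overridden concurrency.
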SoundCue /BarnTowelArmoryGameplay/Assets/Sounds/SC_BarnTowelArmory_Lrg_Arms_DwnForLanding
root=Mixer_0
nodes [
 {id=Delay_0, children [WavePlayer_1], properties {delayMax=0.5, delayMin=0.5}}
 {id=Delay_1, children [WavePlayer_2], properties {delayMax=3.6, delayMin=3.6}}
 {id=Mixer_0, children [Delay_0,Delay_1], properties {inputVolume [1,1.5]}}
 {id=WavePlayer_1, properties {soundWaveAssetPtr=/BarnTowelArmoryGameplay/Assets/Sounds/wav/BarnTowelArmory_Armory_TopArmsComeDown_pt1}}
 {id=WavePlayer_2, properties {soundWaveAssetPtr=/BarnTowelArmoryGameplay/Assets/Sounds/wav/BarnTowelArmory_Armory_TopArmsComeDown_Pt2}}
]
properties {
 attenuationSettings=&/AS_BarnTowel_ShipLanding
 bOverrideConcurrency=true
 maxDistance=53000
 pitchMultiplier=0.95
 soundClassObject=&Fort_Gadgets
 volumeMultiplier=0.7
 concurrencyOverrides {maxCount=2, retriggerTime=2}
}
