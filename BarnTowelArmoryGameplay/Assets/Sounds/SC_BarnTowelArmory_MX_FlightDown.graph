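# Sound cue for the BarnTowelArmory FlightDown music, a one-shot track whose name points to a downward flight sequence.
# A non-looping wave player feeds a fixed one-second delay, wrapped in an enveloper.
# The volume curve drops to silence almost instantly, swells above unity over the first few seconds, then eases slightly by about 19 seconds.
# It uses shared SkyRift music attenuation, the LTM in-game master sound class and a boosted volume multiplier.
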SoundCue /BarnTowelArmoryGameplay/Assets/Sounds/SC_BarnTowelArmory_MX_FlightDown
root=Enveloper_1
nodes [
 {id=Delay_0, children [WavePlayer_1], properties {delayMax=1, delayMin=1}}
 {
  id=Enveloper_1
  children [Delay_0]
  properties {
   pitchCurve {editorCurveData {keys [{time=0}]}}
   volumeCurve {editorCurveData {keys [
      <at 0 1>
      {arriveTangent=0.137, interpMode=RCIM_Cubic, leaveTangent=0.137, tangentMode=RCTM_User, time=0.0001, value=0}
      {arriveTangent=0.9318, interpMode=RCIM_Cubic, leaveTangent=0.9318, time=0.5189, value=0.67}
      {arriveTangent=0.2942, interpMode=RCIM_Cubic, leaveTangent=0.2942, time=1.2896, value=1.2015}
      <at 3 1.4>
      <at 12 1.5>
      <at 19.0654 1.36>
     ]}}
  }
 }
 {id=WavePlayer_1, properties {soundWaveAssetPtr=/BarnTowelArmoryGameplay/Assets/Sounds/wav/BarnTowelArmory_MX_FlightDown_NonLooping}}
]
properties {
 attenuationSettings=&/AS_BarnTowelSkyRift_Music
 maxDistance=91400
 soundClassObject=/Game/Sounds/FortSoundClassesAndModes/Fort_LTM_BR_InGame_Master
 volumeMultiplier=1.7
}
