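# Sound cue for the BarnTowelArmory pod catcher's arm-up sound.
# A fixed short delay feeds a single wave player, with lowered pitch and volume set on the cue.
# It uses the ship-landing attenuation asset and the gadgets sound class, and overrides concurrency to cap simultaneous instances and set a retrigger time.
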SoundCue /BarnTowelArmoryGameplay/Assets/Sounds/SC_BarnTowelArmory_PodCatcher_ArmsUp
root=Delay_0
nodes [
 {id=Delay_0, children [WavePlayer_0], properties {delayMax=0.4, delayMin=0.4}}
 {id=WavePlayer_0, properties {soundWaveAssetPtr=/BarnTowelArmoryGameplay/Assets/Sounds/wav/BarnTowelArmory_Armory_PodCatcher_Arm_Up}}
]
properties {
 attenuationSettings=&/AS_BarnTowel_ShipLanding
 bOverrideConcurrency=true
 maxDistance=53000
 pitchMultiplier=0.9
 soundClassObject=&Fort_Gadgets
 volumeMultiplier=0.66
 concurrencyOverrides {maxCount=4, retriggerTime=2}
}
