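# Sound cue for the BarnTowel Armory Sky Rift activation, playing the rift appear sound.
# A single wave player feeds the output directly, with no randomization or modulation, at a slightly boosted volume.
# It uses rift spawn attenuation with a very large max distance and the gadgets sound class.
# A pre-distance-attenuation send feeds the distant reverb submix, so the reverb send is taken before distance falloff.
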SoundCue /BarnTowelArmoryGameplay/Assets/Sounds/SC_BarnTowelArmory_SkyRift_Activate
root=WavePlayer_2
nodes [{id=WavePlayer_2, properties {soundWaveAssetPtr=/BarnTowelArmoryGameplay/Assets/Sounds/wav/BarnTowelArmory_Armory_SkyRift_Appear}}]
properties {
 attenuationSettings=&/AS_BarnTowel_Rift_Spawn
 maxDistance=300000
 soundClassObject=&Fort_Gadgets
 volumeMultiplier=1.1
 soundSubmixSends [{minSendLevel=0.2, sendLevel=0.2, sendStage=PreDistanceAttenuation, soundSubmix=&Fortnite_Reverb_Distant}]
}
