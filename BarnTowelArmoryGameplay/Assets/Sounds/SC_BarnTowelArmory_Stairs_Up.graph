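# Sound cue for the Barn Towel Armory stairs rising, routed to the Fort_Gadgets sound class.
# A single StairsRise wave passes through a fixed six-second Delay, then an Enveloper whose volume curve dips slightly and recovers over about two seconds.
# Spatialization uses the shared AS_BarnTowel_ShipLanding attenuation asset, and a concurrency override limits simultaneous instances to two.
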SoundCue /BarnTowelArmoryGameplay/Assets/Sounds/SC_BarnTowelArmory_Stairs_Up
root=Enveloper_0
nodes [
 {id=Delay_0, children [WavePlayer_0], properties {delayMax=6, delayMin=6}}
 {
  id=Enveloper_0
  children [Delay_0]
  properties {
   pitchCurve {editorCurveData {keys [{time=0}]}}
   volumeCurve {editorCurveData {keys [
      <at 0 1.3 RCIM_Cubic>
      {arriveTangent=-0.0169, interpMode=RCIM_Cubic, leaveTangent=-0.0169, time=0.7854, value=1.1377}
      <at 2.0088 1.2661 RCIM_Cubic>
     ]}}
  }
 }
 {id=WavePlayer_0, properties {soundWaveAssetPtr=/BarnTowelArmoryGameplay/Assets/Sounds/wav/BarnTowelArmory_Armory_StairsRise}}
]
properties {
 attenuationSettings=&/AS_BarnTowel_ShipLanding
 bOverrideConcurrency=true
 maxDistance=53000
 soundClassObject=&Fort_Gadgets
 volumeMultiplier=1
 attenuationOverrides {
  bAttenuateWithLPF=true
  bEnableOcclusion=true
  falloffDistance=3500
  hPFFrequencyAtMax=80
  lPFFrequencyAtMax=1500
  lPFRadiusMax=2000
  lPFRadiusMin=500
  occlusionLowPassFilterFrequency=1500
  occlusionTraceChannel=ECC_GameTraceChannel6
 }
 concurrencyOverrides {maxCount=2, retriggerTime=2}
}
