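# Sound cue for the BarnTowel Armory thrusters shutting down, classed under Fort_Gadgets.
# A mixer layers the shutdown wave, with an envelope that starts louder and settles over two seconds, with a quieter PodCatcher flap-up wave delayed by a fixed interval.
# Each layer has its own attenuation, ship landing for thrusters and doors for the flap.
# A concurrency override limits simultaneous plays and retriggering.
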SoundCue /BarnTowelArmoryGameplay/Assets/Sounds/SC_BarnTowelArmory_Thrusters_End_Shutdown
root=Mixer_0
nodes [
 {id=Attenuation_0, children [Delay_1], properties {attenuationSettings=&/AS_BarnTowel_Doors}}
 {id=Attenuation_1, children [WavePlayer_1], properties {attenuationSettings=&/AS_BarnTowel_ShipLanding}}
 {id=Delay_1, children [WavePlayer_2], properties {delayMax=1.4, delayMin=1.4}}
 {
  id=Enveloper_0
  children [Attenuation_1]
  properties {pitchCurve {editorCurveData {keys [{time=0}]}}, volumeCurve {editorCurveData {keys [<at 0 1.45>,<at 2 1>]}}}
 }
 {id=Mixer_0, children [Enveloper_0, Attenuation_0], properties {inputVolume [0.66,0.4]}}
 {id=WavePlayer_1, properties {soundWaveAssetPtr=/BarnTowelArmoryGameplay/Assets/Sounds/wav/BarnTowelArmory_Armory_Thrusters_End_Shutdown}}
 {id=WavePlayer_2, properties {soundWaveAssetPtr=/BarnTowelArmoryGameplay/Assets/Sounds/wav/BarnTowelArmory_Armory_PodCatcher_FlapUp}}
]
properties {
 bOverrideConcurrency=true
 maxDistance=53000
 soundClassObject=&Fort_Gadgets
 volumeMultiplier=0.6
 concurrencyOverrides {maxCount=2, retriggerTime=2}
}
