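# Sound cue for the BarnTowel Armory thruster landing sequence, heard over a long range.
# A distance crossfade blends a near mix of 1P start and exhaust waves into a far mix of their 3P counterparts, which also fades out at extreme range.
# Cue-level attenuation overrides add low-pass filtering and a wide falloff, while a concurrency override caps simultaneous instances and sets a retrigger time.
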SoundCue /BarnTowelArmoryGameplay/Assets/Sounds/SC_BarnTowelArmory_Thrusters_LandingSequence
root=DistanceCrossFade_0
nodes [
 {
  id=DistanceCrossFade_0
  children [Mixer_0,Mixer_1]
  properties {crossFadeInput [
    {fadeInDistanceEnd=0, fadeInDistanceStart=0, fadeOutDistanceEnd=15000, fadeOutDistanceStart=4000, volume=1}
    {fadeInDistanceEnd=16000, fadeInDistanceStart=8000, fadeOutDistanceEnd=40000, fadeOutDistanceStart=30000, volume=1}
   ]}
 }
 {id=Mixer_0, children [WavePlayer_0,WavePlayer_3], properties {inputVolume [1,1.1]}}
 {id=Mixer_1, children [WavePlayer_5,WavePlayer_4], properties {inputVolume [1,1]}}
 {id=WavePlayer_0, properties {soundWaveAssetPtr=/BarnTowelArmoryGameplay/Assets/Sounds/wav/BarnTowelArmory_Armory_LandingSeq_1P_Start}}
 {id=WavePlayer_3, properties {soundWaveAssetPtr=/BarnTowelArmoryGameplay/Assets/Sounds/wav/BarnTowelArmory_Armory_LandingSeq_1P_Exhaust}}
 {id=WavePlayer_4, properties {soundWaveAssetPtr=/BarnTowelArmoryGameplay/Assets/Sounds/wav/BarnTowelArmory_Armory_LandingSeq_3P_Exhaust}}
 {id=WavePlayer_5, properties {soundWaveAssetPtr=/BarnTowelArmoryGameplay/Assets/Sounds/wav/BarnTowelArmory_Armory_LandingSeq_3P_Start}}
]
properties {
 attenuationSettings=&/AS_BarnTowel_ShipLanding
 bOverrideAttenuation=true
 bOverrideConcurrency=true
 maxDistance=83000
 soundClassObject=&Fort_SFX_InGame
 volumeMultiplier=1.5
 attenuationOverrides {bApplyNormalizationToStereoSounds=true, bAttenuateWithLPF=true, falloffDistance=80000, stereoSpread=2500, attenuationShapeExtents {x=3000}}
 concurrencyOverrides {maxCount=2, retriggerTime=2}
}
