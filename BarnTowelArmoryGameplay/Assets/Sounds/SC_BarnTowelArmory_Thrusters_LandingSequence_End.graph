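# Sound cue for the end of the BarnTowelArmory thruster landing sequence, playing as an in-game SFX.
# After a short fixed delay, a distance crossfade blends a close-range 1P wave into a distant 3P wave, with a far fade-out on the latter.
# A ship-landing attenuation asset is referenced alongside long-range attenuation overrides with low-pass filtering.
# Concurrency overrides cap simultaneous instances and set a retrigger time.
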SoundCue /BarnTowelArmoryGameplay/Assets/Sounds/SC_BarnTowelArmory_Thrusters_LandingSequence_End
root=Delay_0
nodes [
 {id=Delay_0, children [DistanceCrossFade_0], properties {delayMax=0.2, delayMin=0.2}}
 {
  id=DistanceCrossFade_0
  children [WavePlayer_2,WavePlayer_1]
  properties {crossFadeInput [
    {fadeInDistanceEnd=0, fadeInDistanceStart=0, fadeOutDistanceEnd=15000, fadeOutDistanceStart=4000, volume=1}
    {fadeInDistanceEnd=16000, fadeInDistanceStart=8000, fadeOutDistanceEnd=40000, fadeOutDistanceStart=30000, volume=1}
   ]}
 }
 {id=WavePlayer_1, properties {soundWaveAssetPtr=/BarnTowelArmoryGameplay/Assets/Sounds/wav/BarnTowelArmory_Armory_LandingSeq_3P_End}}
 {id=WavePlayer_2, properties {soundWaveAssetPtr=/BarnTowelArmoryGameplay/Assets/Sounds/wav/BarnTowelArmory_Armory_LandingSeq_1P_End}}
]
properties {
 attenuationSettings=&/AS_BarnTowel_ShipLanding
 bOverrideAttenuation=true
 bOverrideConcurrency=true
 maxDistance=83000
 soundClassObject=&Fort_SFX_InGame
 volumeMultiplier=2
 attenuationOverrides {bApplyNormalizationToStereoSounds=true, bAttenuateWithLPF=true, falloffDistance=80000, stereoSpread=2500, attenuationShapeExtents {x=3000}}
 concurrencyOverrides {maxCount=2, retriggerTime=2}
}
