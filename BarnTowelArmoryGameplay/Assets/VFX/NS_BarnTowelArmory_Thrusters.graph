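# Niagara system for the BarnTowelArmory four-engine downward thruster effect, placed in a cross around the origin.
# Mesh emitters draw flames and glow, ConeLight adds flare sprites and lights, and four GPU emitters spray embers gated by EngineBool.
# System-age curves script ignition, a blue-to-orange flame shift and fade-out.
# Shared System variables drive flare size and opacity, with time-of-day scalars adjusting the brightness of most layers.
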NiagaraSystem /BarnTowelArmoryGameplay/Assets/VFX/NS_BarnTowelArmory_Thrusters
dependencies {
 usedDataInterfaces [&NiagaraDataInterfaceCurve, &NiagaraDataInterfaceColorCurve, &NiagaraDataInterfaceArrayFloat3]
 usedDynamicInputs [&Boolean_AndOperation, &SetBoolByFloatComparison]
 usedModules [
  &SystemState
  &self:SystemSpawnScript.SystemScriptSource.SystemScriptGraph.NiagaraNodeAssignment_0.SetVariables_#1
  &self:SystemSpawnScript.SystemScriptSource.SystemScriptGraph.NiagaraNodeAssignment_1.SetVariables_#2
  &self:SystemSpawnScript.SystemScriptSource.SystemScriptGraph.NiagaraNodeAssignment_2.SetVariables_#3
  &self:SystemSpawnScript.SystemScriptSource.SystemScriptGraph.NiagaraNodeAssignment_3.SetVariables_#4
  &self:SystemSpawnScript.SystemScriptSource.SystemScriptGraph.NiagaraNodeAssignment_4.SetVariables_#5
  &self:SystemSpawnScript.SystemScriptSource.SystemScriptGraph.NiagaraNodeAssignment_5.SetVariables_#6
  &self:SystemSpawnScript.SystemScriptSource.SystemScriptGraph.NiagaraNodeAssignment_6.SetVariables_#7
  &self:SystemSpawnScript.SystemScriptSource.SystemScriptGraph.NiagaraNodeAssignment_7.SetVariables_#8
  &self:SystemSpawnScript.SystemScriptSource.SystemScriptGraph.NiagaraNodeAssignment_8.SetVariables_#9
  &self:SystemSpawnScript.SystemScriptSource.SystemScriptGraph.NiagaraNodeAssignment_9.SetVariables_#10
 ]
}
emitters {
 Center_flame {
  emitterUpdateScript [
   &EmitterState()
   &SpawnRate(SpawnRate=30, Use Spawn Probability=false) {enabled=false}
   &SpawnBurst_Instantaneous(Spawn Count=4, Use Spawn Probability=true)
  ]
  particleSpawnScript [
   &InitializeParticle(
    Lifetime=0.3
    Mesh Scale=&RandomRangeVector(Minimum=<0.24 0.24 0.3>, Maximum=<0.2 0.2 0.2>)
    Mesh Scale Mode=::NewEnumerator4
    UsePositionOffset=false
    Write Material Random=true
   ) {module=InitializeParticle001}
   &PartitionParticles(Number Of Partitions=4)
   &DynamicMaterialParameters(
    Index 0 Param 3=15
    Module.Param0WriteEnabled=false
    Module.Param1WriteEnabled=false
    Module.Param2WriteEnabled=true
    Module.Param3WriteEnabled=false
   )
   &CameraOffset(Camera Offset Amount=10)
   &SystemLocation(
    MaskSpawnGroup=false
    Offset=&SelectVectorFromArray(
     Direct Array Index=$Particles.PartitionParticles.Partition
     Vector Selection Array=DataInterface() {propertyValues {FloatData [<266 0 30>, <-266 0 30>, <0 266 30>, <0 -266 30>]}}
    )
   )
   &InitialMeshOrientation(Mesh Axis To Orient=<0 0 1>, Mesh Orientation Mode=::NewEnumerator2, Orientation Vector=<0 0 -1>, Use Rotation Vector=false)
  ]
  particleUpdateScript [
   &ParticleState(Kill Particles When Lifetime Has Elapsed=false, Let Infinitely Lived Particles Die When Emitter Deactivates=true)
   &ScaleMeshSize(Scale Factor=&Multiply_VectorByFloat(
     Vector=&SelectVectorFromArray(
      Direct Array Index=$Particles.SpawnGroup
      Vector Selection Array=DataInterface() {propertyValues {FloatData [<2.5 2.5 1.5>, <2.5 2.5 1.5>, <2.5 2.5 1.5>, <2.5 2.5 1.5>]}}
     )
     Float=&FloatFromCurve(FloatCurve=DataInterface() {propertyValues {Curve {keys [<at 0 0>,<at 0.15 1>,<at 0.5 1>]}}})
    )) {module=ScaleMeshSize001}
   &SolveForcesAndVelocity(Clamp Velocity=false, Limit Acceleration=false)
   &ScaleColor(
    Scale RGB=&Multiply_VectorByFloat(
     Vector=&MakeVectorFromLinearColorRGB(Color=&ColorFromCurve(
       ColorCurve=DataInterface() {propertyValues {
         AlphaCurve {keys []}
         BlueCurve {keys [<at 0.28 0 RCIM_Cubic>]}
         GreenCurve {keys [<at 0.28 0.1651 RCIM_Cubic>]}
         RedCurve {keys [<at 0.28 1 RCIM_Cubic>]}
        }}
       CurveIndex=$Emitter.Age
      ))
     Float=0.5
    )
    ScaleA=true
    ScaleRGB=true
   ) {enabled=false}
   &ScaleMeshSize(Scale Factor=&Multiply_VectorByFloat(
     Vector=&MakeVector(
      X=&Multiply_Float(
       A=0.7
       B=&RemapRange(Input Value To Remap=$System.AirConeFlareMult, Input Min=1, Input Max=4, Output Min=1, Output Max=1.8)
      )
      Y=&Multiply_Float(
       A=0.7
       B=&RemapRange(Input Value To Remap=$System.AirConeFlareMult, Input Min=1, Input Max=4, Output Min=1, Output Max=1.8)
      )
      Z=2
     )
     Float=10
    ))
   &ScaleMeshSize(Scale Factor=&MakeVector(
     X=&Multiply_Float(
      A=&RandomRangeFloat(Minimum=0.7, Maximum=1.3, Evaluation Type=::NewEnumerator1)
      B=$System.ThrustersFlare
     )
     Y=&Multiply_Float(
      A=&RandomRangeFloat(Minimum=0.7, Maximum=1.3, Evaluation Type=::NewEnumerator1)
      B=$System.ThrustersFlare
     )
     Z=&Multiply_Float(
      A=&FloatFromCurve(
       FloatCurve=DataInterface() {propertyValues {Curve {keys [
           <at 8 0.35 RCIM_Cubic>
           {arriveTangent=0.4333, leaveTangent=0.4333, time=9, value=1}
           {arriveTangent=-0.0606, interpMode=RCIM_Cubic, leaveTangent=-0.0606, time=23, value=1}
           {interpMode=RCIM_Cubic, leaveTangent=0.0025, time=25.5, value=0}
          ]}}}
       CurveIndex=$System.Age
      )
      B=&RemapRange(Input Value To Remap=$System.AirConeFlareMult, Input Min=1, Input Max=4, Output Min=1, Output Max=3)
     )
    )) {module=ScaleMeshSize002}
   &Color(
    Color=&LerpLinearColors(
     StartColor=<rgba 0 0.1221 1 1>
     EndColor=<rgba 1 0.1651 0 1>
     LerpFactor=&FloatFromCurve(FloatCurve=DataInterface() {propertyValues {Curve {keys [<at 13 0 RCIM_Cubic>,<at 14 1 RCIM_Cubic>]}}}, CurveIndex=$Engine.System.Age)
    )
    Scale Alpha=&FloatFromCurve(
     FloatCurve=DataInterface() {propertyValues {Curve {keys [
         <at 8 0.35 RCIM_Cubic>
         {arriveTangent=-0.0179, time=10, value=1}
         {arriveTangent=-0.0952, interpMode=RCIM_Cubic, leaveTangent=-0.0952, time=19, value=1}
         <at 20.5 0 RCIM_Cubic>
        ]}}}
     CurveIndex=$Engine.System.Age
    )
    Scale Color=&VectorFromFloat(Value=&FloatFromCurve(FloatCurve=DataInterface() {propertyValues {Curve {keys [<at 13 2 RCIM_Cubic>,<at 14 1 RCIM_Cubic>]}}}, CurveIndex=$Engine.System.Age))
   )
   /Game/Weapons/FORT_Melee/Pickaxe_Pumpkin_Spice_1h/FX/ColorFromTOD(FortniteNPC_FortniteActiveTimeOfDay:LinearColor=$NPC.FortniteNPC.FortniteActiveTimeOfDay, Night:LinearColor=<rgba 0.3596 1 0 1>) {enabled=false}
   &ScaleColor(
    Color Value To Scale=$Particles.Color
    Scale Alpha=&FloatFromCurve(
     FloatCurve=DataInterface() {propertyValues {Curve {keys [
         <at 5 3 RCIM_Cubic>
         {arriveTangent=-0.4444, leaveTangent=-0.4444, time=8, value=1}
         {arriveTangent=1, interpMode=RCIM_Cubic, leaveTangent=1, time=9.5, value=1}
         <at 10 3 RCIM_Cubic>
         <at 11 1 RCIM_Cubic>
        ]}}}
     CurveIndex=$Engine.System.Age
     Scale Curve=&Multiply_Float(A=1, B=&FNTimeOfDayScalar(Night Scale=0.15, Morning Scale=1, Evening Scale=1))
    )
    ScaleA=true
    ScaleRGB=true
   ) {module=ScaleColor001}
  ]
  rendererClasses [&NiagaraMeshRendererProperties]
  renderers [{rendererClass=&NiagaraMeshRendererProperties, rendererIndex=0}]
 }
 ConeLight {
  emitterSpawnScript [
   /SetVariables(Emitter.Distortion_Input=1, Emitter.LightConeOverallSize=2.2) {isSetParameters=true}
   /SetVariables(Emitter.FlareSizeScalar=0.3) {isSetParameters=true}
  ]
  emitterUpdateScript [
   &EmitterState(Life Cycle Mode=::NewEnumerator1, Loop Behavior=::NewEnumerator1, Loop Duration Mode=::NewEnumerator1, UseLoopDelay=false)
   &SpawnBurst_Instantaneous(Spawn Count=4, Use Spawn Probability=false)
  ]
  particleSpawnScript [
   &InitializeParticle(
    Color=<rgba 0.7804 0.298 0 1>
    Lifetime=5
    Sprite Size=&Multiply_Vector2D_ByFloat(Float=$Emitter.FlareSizeScalar, Vector2D=<15000 10000>)
    Sprite Size Mode=::NewEnumerator4
    UsePositionOffset=false
    Write Material Random=true
   )
   &PartitionParticles(Number Of Partitions=4)
   &CameraOffset(Camera Offset Amount=200)
   /SetVariables(Particles.LightPos:NiagaraPosition=&AddVectorToPosition(Position=$Particles.Position, Vector=<0 0 -60>)) {isSetParameters=true}
  ]
  particleUpdateScript [
   &ParticleState(Kill Particles When Lifetime Has Elapsed=false, Let Infinitely Lived Particles Die When Emitter Deactivates=true)
   &Color(Color=&LerpLinearColors(
     StartColor=&MultiplyLinearColor(Linear Color=<rgba 0 0.03 0.15 1>, Scale Factor=&MakeVector4(X=1, Y=1, Z=1, W=3))
     EndColor=$System.LerpColor
     LerpFactor=&FloatFromCurve(
      FloatCurve=DataInterface() {propertyValues {Curve {keys [{tangentMode=RCTM_User, time=6, value=0}, {tangentMode=RCTM_User, time=8, value=1}]}}}
      CurveIndex=$Engine.System.Age
     )
    ))
   &SystemLocation(
    MaskSpawnGroup=false
    Offset=&SelectVectorFromArray(
     Direct Array Index=$Particles.PartitionParticles.Partition
     Vector Selection Array=DataInterface() {propertyValues {FloatData [<266 0 -100>, <-266 0 -100>, <0 266 -100>, <0 -266 -100>]}}
    )
   )
   /SetVariables(Particles.Sputter:NiagaraFloat=&RandomRangeFloat(Minimum=0.8, Maximum=1, Evaluation Type=::NewEnumerator1)) {isSetParameters=true}
   /SetVariables(
    Particles.LensColor:LinearColor=&MultiplyLinearColor(
     Linear Color=<rgba 0.2 0.017 0 1>
     Scale Factor=&MakeVector4_FromFloat(Float=&Multiply_Float(A=$System.FlareOpacity, B=$Particles.Sputter))
    )
    Particles.LensSize:Vector2f=&Multiply_Vector2D_ByFloat(
     Float=&Multiply_Float(A=$Emitter.FlareSizeScalar, B=1)
     Vector2D=&Multiply_Vector2D_ByFloat(Float=&Multiply_Float(A=0.4, B=$System.AirConeFlareMult), Vector2D=<8000 4000>)
    )
   ) {isSetParameters=true}
   /SetVariables(
    Particles.DarkRadialColor:LinearColor=&MultiplyLinearColor(
     Linear Color=<rgba 0.0902 0.0318 0.038 0.5>
     Scale Factor=&MakeVector4_FromFloat(Float=&Multiply_Float(A=$System.FlareOpacity, B=$Particles.Sputter))
    )
    Particles.DarkRadialSize:Vector2f=&Multiply_Vector2D_ByFloat(
     Float=&Multiply_Float(A=$Emitter.FlareSizeScalar, B=1)
     Vector2D=&Multiply_Vector2D_ByFloat(Float=&Multiply_Float(A=$System.AirConeFlareMult, B=0.2), Vector2D=<10000 10000>)
    )
   ) {isSetParameters=true}
   /SetVariables(Particles.LightColor=&MultiplyLinearColor(
     Linear Color=$Particles.Color
     Scale Factor=&MakeVector4_FromFloat(Float=&Multiply_Float(
       A=&Multiply_Float(A=&Multiply_Float(A=8, B=$System.FlareOpacity), B=$System.EnginesActive)
       B=&Multiply_Float(A=1, B=$Particles.Sputter)
      ))
    )) {isSetParameters=true}
   &ScaleSpriteSize(Scale Factor=&Multiply_Vector2D_ByFloat(
     Float=&Multiply_Float(
      A=$Emitter.FlareSizeScalar
      B=&Multiply_Float(A=&Multiply_Float(A=1.25, B=$System.AirConeFlareMult), B=$System.FlareOpacity)
     )
     Vector2D=&Multiply_Vector2D(
      A=&Vector2DFromFloat(Value=&Lerp_Float(A=1, B=&RandomRangeFloat(Minimum=0.8, Maximum=1, Evaluation Type=::NewEnumerator1), Alpha=1))
      B=&MakeVector2D(
       X=&Lerp_Float(
        A=1
        B=1.5
        Alpha=&FloatFromCurve(
         FloatCurve=DataInterface() {propertyValues {Curve {keys [<at 8 0>,<at 9 1>,<at 9.5 1 RCIM_Cubic>,<at 10 1 RCIM_Cubic>,<at 11 1>]}}}
         CurveIndex=$System.Age
        )
       )
       Y=1
      )
     )
    ))
   &ScaleColor(
    Color Value To Scale=$Particles.Color
    Scale Alpha=&Multiply_Float(
     A=&Multiply_Float(
      A=&Multiply_Float(A=0.5, B=&FNTimeOfDayScalar(Night Scale=0.3, Morning Scale=1, Evening Scale=1))
      B=$System.FlareOpacity
     )
     B=$Particles.Sputter
    )
    Scale RGB=&VectorFromFloat(Value=&FloatFromCurve(FloatCurve=DataInterface() {propertyValues {Curve {keys [{interpMode=RCIM_Cubic, tangentMode=RCTM_User, time=0, value=3}, {interpMode=RCIM_Cubic, tangentMode=RCTM_User, time=0.35, value=10}]}}}))
    ScaleA=true
    ScaleRGB=true
   )
  ]
  rendererClasses [&NiagaraSpriteRendererProperties, &NiagaraLightRendererProperties]
  renderers [
   {rendererClass=&NiagaraSpriteRendererProperties}
   {rendererClass=&NiagaraSpriteRendererProperties, rendererIndex=1}
   {rendererClass=&NiagaraSpriteRendererProperties, rendererIndex=2}
   {rendererClass=&NiagaraLightRendererProperties, rendererIndex=3}
  ]
 }
 Embers {
  simTarget=GPUComputeSim
  emitterUpdateScript [
   &EmitterState()
   &SpawnRate(
    Spawn Probability=0.5
    SpawnRate=&MakeCustomFloatFromBool(Bool=$System.EngineBool, True Float=$System.EmbersSpawnRate, False Float=0)
    Use Spawn Probability=true
   )
  ]
  particleSpawnScript [
   &InitializeParticle(
    Color Maximum=<rgba 25 2 0 1>
    Color Minimum=<rgba 10 2 0 1>
    Color Mode=::NewEnumerator2
    Lifetime=&RandomRangeFloat(Minimum=0.17, Maximum=0.4)
    Sprite Size Mode=::NewEnumerator3
    Uniform Sprite Size=6
    UsePositionOffset=false
    Write Material Random=true
   )
   &ShapeLocation(
    Cylinder Height=1
    Cylinder Radius=80
    MaskSpawnGroup=false
    Offset=<266 0 0>
    Offset Mode=::NewEnumerator0
    Override Seed=false
    Shape Primitive=::NewEnumerator1
    Surface Only=true
    Use Endcaps In Surface Only Mode=false
    UseOwnerScale=true
   )
   &AddVelocity(
    Cone Angle=10
    Cone Axis=<0 0 -1>
    Module.UseConeFalloff=false
    Module.UseVelDistribution=false
    Override Seed=false
    Velocity Mode=::NewEnumerator2
    Velocity Speed=&Multiply_Float(A=&RandomRangeFloat(Minimum=512, Maximum=1024), B=7)
   )
  ]
  particleUpdateScript [
   &ParticleState()
   &Color(Color=&LerpLinearColors(
     StartColor=<rgba 10 10 10 1>
     EndColor=&RandomRangeLinearColor(Minimum=<rgba 10 2 0 1>, Maximum=<rgba 25 2 0 1>)
     LerpFactor=&FloatFromCurve(FloatCurve=DataInterface() {propertyValues {Curve {keys [<at 13 0 RCIM_Cubic>,<at 14 1 RCIM_Cubic>]}}}, CurveIndex=$Engine.System.Age)
    ))
   &ScaleColor(
    Color Value To Scale=$Particles.Color
    Scale Alpha=&FloatFromCurve(
     FloatCurve=DataInterface() {propertyValues {Curve {keys [<at -0.0046 1.0113 RCIM_Cubic>,<at 1.0011 0.002 RCIM_Cubic>]}}}
     Scale Curve=&FNTimeOfDayScalar(Night Scale=0.3, Morning Scale=1, Evening Scale=1)
    )
    ScaleA=true
    ScaleRGB=true
   )
   &Drag(Drag=&RandomRangeFloat(Minimum=2, Maximum=4), Use Linear Drag=true, Use Rotational Drag=true)
   &CurlNoiseForce(Noise Frequency=10, Noise Strength=50, Pan Noise=true, Pan Noise Field=<1 1069.1512 0>)
   &SolveForcesAndVelocity(Clamp Velocity=false, Limit Acceleration=false)
   &ScaleSpriteSizeBySpeed(Max Scale Factor=<1 8>, Min Scale Factor=<1 0>)
  ]
  rendererClasses [&NiagaraSpriteRendererProperties]
  renderers [{rendererClass=&NiagaraSpriteRendererProperties}]
 }
 Embers001 {
  simTarget=GPUComputeSim
  emitterUpdateScript [
   &EmitterState()
   &SpawnRate(
    Spawn Probability=0.5
    SpawnRate=&MakeCustomFloatFromBool(Bool=$System.EngineBool, True Float=$System.EmbersSpawnRate, False Float=0)
    Use Spawn Probability=true
   )
  ]
  particleSpawnScript [
   &InitializeParticle(
    Color Maximum=<rgba 25 2 0 1>
    Color Minimum=<rgba 10 2 0 1>
    Color Mode=::NewEnumerator2
    Lifetime=&RandomRangeFloat(Minimum=0.17, Maximum=0.4)
    Sprite Size Mode=::NewEnumerator3
    Uniform Sprite Size=6
    UsePositionOffset=false
    Write Material Random=true
   )
   &AddVelocity(
    Cone Angle=10
    Cone Axis=<0 0 -1>
    Module.UseConeFalloff=false
    Module.UseVelDistribution=false
    Override Seed=false
    Velocity Mode=::NewEnumerator2
    Velocity Speed=&Multiply_Float(A=&RandomRangeFloat(Minimum=512, Maximum=1024), B=7)
   )
   &ShapeLocation(
    Cylinder Height=1
    Cylinder Radius=80
    MaskSpawnGroup=false
    Offset=<-266 0 0>
    Offset Mode=::NewEnumerator0
    Override Seed=false
    Shape Primitive=::NewEnumerator1
    Surface Only=true
    Use Endcaps In Surface Only Mode=false
    UseOwnerScale=true
   ) {module=ShapeLocation001}
  ]
  particleUpdateScript [
   &ParticleState()
   &Drag(Drag=&RandomRangeFloat(Minimum=2, Maximum=4), Use Linear Drag=true, Use Rotational Drag=true)
   &CurlNoiseForce(Noise Frequency=10, Noise Strength=50, Pan Noise=true, Pan Noise Field=<1 1069.1512 0>)
   &SolveForcesAndVelocity(Clamp Velocity=false, Limit Acceleration=false)
   &ScaleSpriteSizeBySpeed(Max Scale Factor=<1 8>, Min Scale Factor=<1 0>)
   &Color(Color=&LerpLinearColors(
     StartColor=<rgba 10 10 10 1>
     EndColor=&RandomRangeLinearColor(Minimum=<rgba 10 2 0 1>, Maximum=<rgba 25 2 0 1>)
     LerpFactor=&FloatFromCurve(FloatCurve=DataInterface() {propertyValues {Curve {keys [<at 13 0 RCIM_Cubic>,<at 14 1 RCIM_Cubic>]}}}, CurveIndex=$Engine.System.Age)
    ))
   &ScaleColor(
    Color Value To Scale=$Particles.Color
    Scale Alpha=&FloatFromCurve(
     FloatCurve=DataInterface() {propertyValues {Curve {keys [<at -0.0046 1.0113 RCIM_Cubic>,<at 1.0011 0.002 RCIM_Cubic>]}}}
     Scale Curve=&FNTimeOfDayScalar(Night Scale=0.3, Morning Scale=1, Evening Scale=1)
    )
    ScaleA=true
    ScaleRGB=true
   )
  ]
  rendererClasses [&NiagaraSpriteRendererProperties]
  renderers [{rendererClass=&NiagaraSpriteRendererProperties}]
 }
 Embers002 {
  simTarget=GPUComputeSim
  emitterUpdateScript [
   &EmitterState()
   &SpawnRate(
    Spawn Probability=0.5
    SpawnRate=&MakeCustomFloatFromBool(Bool=$System.EngineBool, True Float=$System.EmbersSpawnRate, False Float=0)
    Use Spawn Probability=true
   )
  ]
  particleSpawnScript [
   &InitializeParticle(
    Color Maximum=<rgba 25 2 0 1>
    Color Minimum=<rgba 10 2 0 1>
    Color Mode=::NewEnumerator2
    Lifetime=&RandomRangeFloat(Minimum=0.17, Maximum=0.4)
    Sprite Size Mode=::NewEnumerator3
    Uniform Sprite Size=6
    UsePositionOffset=false
    Write Material Random=true
   )
   &AddVelocity(
    Cone Angle=10
    Cone Axis=<0 0 -1>
    Module.UseConeFalloff=false
    Module.UseVelDistribution=false
    Override Seed=false
    Velocity Mode=::NewEnumerator2
    Velocity Speed=&Multiply_Float(A=&RandomRangeFloat(Minimum=512, Maximum=1024), B=7)
   )
   &ShapeLocation(
    Cylinder Height=1
    Cylinder Radius=80
    MaskSpawnGroup=false
    Offset=<0 266 0>
    Offset Mode=::NewEnumerator0
    Override Seed=false
    Shape Primitive=::NewEnumerator1
    Surface Only=true
    Use Endcaps In Surface Only Mode=false
    UseOwnerScale=true
   ) {module=ShapeLocation001}
  ]
  particleUpdateScript [
   &ParticleState()
   &Drag(Drag=&RandomRangeFloat(Minimum=2, Maximum=4), Use Linear Drag=true, Use Rotational Drag=true)
   &CurlNoiseForce(Noise Frequency=10, Noise Strength=50, Pan Noise=true, Pan Noise Field=<1 1069.1512 0>)
   &SolveForcesAndVelocity(Clamp Velocity=false, Limit Acceleration=false)
   &ScaleSpriteSizeBySpeed(Max Scale Factor=<1 8>, Min Scale Factor=<1 0>)
   &Color(Color=&LerpLinearColors(
     StartColor=<rgba 10 10 10 1>
     EndColor=&RandomRangeLinearColor(Minimum=<rgba 10 2 0 1>, Maximum=<rgba 25 2 0 1>)
     LerpFactor=&FloatFromCurve(FloatCurve=DataInterface() {propertyValues {Curve {keys [<at 13 0 RCIM_Cubic>,<at 14 1 RCIM_Cubic>]}}}, CurveIndex=$Engine.System.Age)
    ))
   &ScaleColor(
    Color Value To Scale=$Particles.Color
    Scale Alpha=&FloatFromCurve(
     FloatCurve=DataInterface() {propertyValues {Curve {keys [<at -0.0046 1.0113 RCIM_Cubic>,<at 1.0011 0.002 RCIM_Cubic>]}}}
     Scale Curve=&FNTimeOfDayScalar(Night Scale=0.3, Morning Scale=1, Evening Scale=1)
    )
    ScaleA=true
    ScaleRGB=true
   )
  ]
  rendererClasses [&NiagaraSpriteRendererProperties]
  renderers [{rendererClass=&NiagaraSpriteRendererProperties}]
 }
 Embers003 {
  simTarget=GPUComputeSim
  emitterUpdateScript [
   &EmitterState()
   &SpawnRate(
    Spawn Probability=0.5
    SpawnRate=&MakeCustomFloatFromBool(Bool=$System.EngineBool, True Float=$System.EmbersSpawnRate, False Float=0)
    Use Spawn Probability=true
   )
  ]
  particleSpawnScript [
   &InitializeParticle(
    Color Maximum=<rgba 25 2 0 1>
    Color Minimum=<rgba 10 2 0 1>
    Color Mode=::NewEnumerator2
    Lifetime=&RandomRangeFloat(Minimum=0.17, Maximum=0.4)
    Sprite Size Mode=::NewEnumerator3
    Uniform Sprite Size=6
    UsePositionOffset=false
    Write Material Random=true
   )
   &AddVelocity(
    Cone Angle=10
    Cone Axis=<0 0 -1>
    Module.UseConeFalloff=false
    Module.UseVelDistribution=false
    Override Seed=false
    Velocity Mode=::NewEnumerator2
    Velocity Speed=&Multiply_Float(A=&RandomRangeFloat(Minimum=512, Maximum=1024), B=7)
   )
   &ShapeLocation(
    Cylinder Height=1
    Cylinder Radius=80
    MaskSpawnGroup=false
    Offset=<0 -266 0>
    Offset Mode=::NewEnumerator0
    Override Seed=false
    Shape Primitive=::NewEnumerator1
    Surface Only=true
    Use Endcaps In Surface Only Mode=false
    UseOwnerScale=true
   ) {module=ShapeLocation001}
  ]
  particleUpdateScript [
   &ParticleState()
   &Drag(Drag=&RandomRangeFloat(Minimum=2, Maximum=4), Use Linear Drag=true, Use Rotational Drag=true)
   &CurlNoiseForce(Noise Frequency=10, Noise Strength=50, Pan Noise=true, Pan Noise Field=<1 1069.1512 0>)
   &SolveForcesAndVelocity(Clamp Velocity=false, Limit Acceleration=false)
   &ScaleSpriteSizeBySpeed(Max Scale Factor=<1 8>, Min Scale Factor=<1 0>)
   &Color(Color=&LerpLinearColors(
     StartColor=<rgba 10 10 10 1>
     EndColor=&RandomRangeLinearColor(Minimum=<rgba 10 2 0 1>, Maximum=<rgba 25 2 0 1>)
     LerpFactor=&FloatFromCurve(FloatCurve=DataInterface() {propertyValues {Curve {keys [<at 13 0 RCIM_Cubic>,<at 14 1 RCIM_Cubic>]}}}, CurveIndex=$Engine.System.Age)
    ))
   &ScaleColor(
    Color Value To Scale=$Particles.Color
    Scale Alpha=&FloatFromCurve(
     FloatCurve=DataInterface() {propertyValues {Curve {keys [<at -0.0046 1.0113 RCIM_Cubic>,<at 1.0011 0.002 RCIM_Cubic>]}}}
     Scale Curve=&FNTimeOfDayScalar(Night Scale=0.3, Morning Scale=1, Evening Scale=1)
    )
    ScaleA=true
    ScaleRGB=true
   )
  ]
  rendererClasses [&NiagaraSpriteRendererProperties]
  renderers [{rendererClass=&NiagaraSpriteRendererProperties}]
 }
 Glow001 {
  emitterUpdateScript [
   &EmitterState()
   &SpawnBurst_Instantaneous(Spawn Count=4, Use Spawn Probability=false, UseLoopCountLimit=true)
  ]
  particleSpawnScript [
   &InitializeParticle(
    Color=<rgba 0.5 0.0576 0.026 1>
    Lifetime=0.5
    Mesh Scale=<0.7 0.7 0.7>
    Mesh Scale Mode=::NewEnumerator4
    Sprite Rotation Angle=0.25
    Sprite Rotation Mode=::NewEnumerator1
    Sprite Size=<90 50>
    Sprite Size Mode=::NewEnumerator4
    UsePositionOffset=false
    Write Material Random=true
   )
   &CameraOffset()
   &PartitionParticles(Number Of Partitions=4)
   &InitialMeshOrientation(Mesh Axis To Orient=<0 0 1>, Mesh Orientation Mode=::NewEnumerator2, Orientation Vector=<0 0 -1>, Use Rotation Vector=false)
   &SystemLocation(
    MaskSpawnGroup=false
    Offset=&SelectVectorFromArray(
     Direct Array Index=$Particles.PartitionParticles.Partition
     Vector Selection Array=DataInterface() {propertyValues {FloatData [<266 0 -50>, <-266 0 -50>, <0 266 -50>, <0 -266 -50>]}}
    )
   )
  ]
  particleUpdateScript [
   &ParticleState(Kill Particles When Lifetime Has Elapsed=false, Let Infinitely Lived Particles Die When Emitter Deactivates=true)
   &Color(Color=&LerpLinearColors(
     StartColor=<rgba 0 0.1221 1 1>
     EndColor=<rgba 0.3333 0.055 0 1>
     LerpFactor=&FloatFromCurve(FloatCurve=DataInterface() {propertyValues {Curve {keys [<at 13 0 RCIM_Cubic>,<at 14 1 RCIM_Cubic>]}}}, CurveIndex=$Engine.System.Age)
    ))
   &ScaleMeshSize(Scale Factor=&MakeVector(
     X=&Multiply_Float(A=0.7, B=$System.ThrustersFlare)
     Y=&Multiply_Float(A=0.7, B=$System.ThrustersFlare)
     Z=&FloatFromCurve(
      FloatCurve=DataInterface() {propertyValues {Curve {keys [
          <at 0 0 RCIM_Cubic>
          {arriveTangent=0.0125, interpMode=RCIM_Cubic, leaveTangent=0.0125, time=3, value=0.1}
          {arriveTangent=0.0144, interpMode=RCIM_Cubic, leaveTangent=0.0144, tangentMode=RCTM_User, time=8, value=0.1}
          {arriveTangent=1.2086, interpMode=RCIM_Cubic, leaveTangent=1.2086, tangentMode=RCTM_User, time=9, value=1}
         ]}}}
      CurveIndex=$System.Age
     )
    )) {module=ScaleMeshSize002}
   &SolveForcesAndVelocity(Clamp Velocity=false, Limit Acceleration=false)
   &DynamicMaterialParameters(
    Index 0 Param 1=2
    Index 0 Param 2=150
    Index 0 Param 3=1
    Index 0 Param 4=1
    Module.Param0WriteEnabled=true
    Module.Param1WriteEnabled=true
    Module.Param2WriteEnabled=true
    Module.Param3WriteEnabled=true
   )
   &ScaleMeshSize(Scale Factor=&Multiply_VectorByFloat(Vector=<1 1 1>, Float=10)) {module=ScaleMeshSize001}
   &ScaleColor(
    Color Value To Scale=$Particles.Color
    Scale Alpha=&FloatFromCurve(FloatCurve=DataInterface() {propertyValues {Curve {keys [
         <at 0 0 RCIM_Cubic>
         {arriveTangent=-0.0606, interpMode=RCIM_Cubic, leaveTangent=-0.0606, tangentMode=RCTM_User, time=0.5, value=1}
         <at 1 1 RCIM_Cubic>
        ]}}})
    ScaleA=true
    ScaleRGB=true
   )
   &ScaleColor(
    Color Value To Scale=$Particles.Color
    Scale Alpha=&FloatFromCurve(
     FloatCurve=DataInterface() {propertyValues {Curve {keys [
         <at 0 0 RCIM_Cubic>
         {arriveTangent=0.0032, interpMode=RCIM_Cubic, leaveTangent=0.0032, tangentMode=RCTM_User, time=3, value=0.5}
         {arriveTangent=-0.0012, interpMode=RCIM_Cubic, leaveTangent=-0.0012, tangentMode=RCTM_User, time=8, value=0.5}
         {arriveTangent=-0.0179, time=9, value=1}
         {arriveTangent=-0.087, interpMode=RCIM_Cubic, leaveTangent=-0.087, time=19, value=1}
         <at 20.5 0 RCIM_Cubic>
        ]}}}
     CurveIndex=$Engine.System.Age
     Scale Curve=0.5
    )
    ScaleA=true
    ScaleRGB=true
   ) {module=ScaleColor001}
  ]
  rendererClasses [&NiagaraMeshRendererProperties]
  renderers [{rendererClass=&NiagaraMeshRendererProperties, rendererIndex=0}]
 }
 Thrusters {
  emitterUpdateScript [
   &EmitterState()
   &SpawnBurst_Instantaneous(Spawn Count=4, Use Spawn Probability=false, UseLoopCountLimit=false)
   &SpawnRate(SpawnRate=4, Use Spawn Probability=false) {enabled=false}
  ]
  particleSpawnScript [
   &PartitionParticles(Number Of Partitions=4)
   &InitializeParticle(
    Color=<rgba 0.9583 0.4145 0 1>
    Mesh Scale=<0.05 0.075 0.075>
    Mesh Scale Mode=::NewEnumerator4
    Sprite Size Mode=::NewEnumerator3
    Uniform Sprite Size=25
    UsePositionOffset=false
    Write Material Random=true
   )
   /SetVariables(Particles.Dynam_Param2:Vector4f=<1 5 5 5>) {isSetParameters=true}
   &InitialMeshOrientation(Mesh Axis To Orient=<-1 0 0>, Mesh Orientation Mode=::NewEnumerator2, Orientation Vector=<0 0 -1>, Use Rotation Vector=false) {module=InitialMeshOrientation001}
   &DynamicMaterialParameters(
    Index 0 Param 1=2
    Index 0 Param 2=5
    Index 0 Param 3=5
    Index 0 Param 4=5
    Module.Index0Param0Write=true
    Module.Index0Param1Write=true
    Module.Index0Param2Write=true
    Module.Index0Param3Write=true
   )
   &SystemLocation(
    MaskSpawnGroup=false
    Offset=&SelectVectorFromArray(
     Direct Array Index=$Particles.PartitionParticles.Partition
     Vector Selection Array=DataInterface() {propertyValues {FloatData [<266 0 -50>, <-266 0 -50>, <0 266 -50>, <0 -266 -50>]}}
    )
   )
  ]
  particleUpdateScript [
   &ParticleState(Kill Particles When Lifetime Has Elapsed=false)
   &ScaleMeshSize(Scale Factor=&Multiply_VectorByFloat(
     Vector=&MakeVector(
      X=&FloatFromCurve(
       FloatCurve=DataInterface() {propertyValues {Curve {keys [<at 10 0.25 RCIM_Cubic>, {arriveTangent=1.2086, interpMode=RCIM_Cubic, leaveTangent=1.2086, tangentMode=RCTM_User, time=12, value=1}]}}}
       CurveIndex=$System.Age
      )
      Y=&Multiply_Float(A=$System.EnginesActive, B=$System.ThrustersFlare)
      Z=&Multiply_Float(A=$System.EnginesActive, B=$System.ThrustersFlare)
     )
     Float=60
    )) {enabled=false, module=ScaleMeshSize002}
   &ScaleMeshSize(Scale Factor=&Multiply_VectorByFloat(
     Vector=&MakeVector(
      X=&Multiply_Float(
       A=&FloatFromCurve(
        FloatCurve=DataInterface() {propertyValues {Curve {keys [
            <at 8 0.35 RCIM_Cubic>
            {arriveTangent=1.2086, time=9, value=1}
            {arriveTangent=-0.087, interpMode=RCIM_Cubic, leaveTangent=-0.087, time=18, value=1}
            <at 20.5 0 RCIM_Cubic>
           ]}}}
        CurveIndex=$System.Age
       )
       B=&RemapRange(Input Value To Remap=$System.AirConeFlareMult, Input Min=1, Input Max=4, Output Min=1, Output Max=3)
      )
      Y=&Multiply_Float(
       A=&Multiply_Float(
        A=&RandomRangeFloat(Minimum=1, Maximum=1.4, Evaluation Type=::NewEnumerator1)
        B=$System.ThrustersFlare_Outer
       )
       B=&RemapRange(Input Value To Remap=$System.AirConeFlareMult, Input Min=1, Input Max=4, Output Min=1, Output Max=1.35)
      )
      Z=&Multiply_Float(
       A=&Multiply_Float(
        A=&RandomRangeFloat(Minimum=1, Maximum=1.4, Evaluation Type=::NewEnumerator1)
        B=$System.ThrustersFlare_Outer
       )
       B=&RemapRange(Input Value To Remap=$System.AirConeFlareMult, Input Min=1, Input Max=4, Output Min=1, Output Max=1.35)
      )
     )
     Float=70
    )) {module=ScaleMeshSize001}
   /SetVariables(Particles.InnerFlame_Color:LinearColor=&MultiplyLinearColor(
     Linear Color=<rgba 1 1 1 1>
     Scale Factor=&MakeVector4_FromFloat(Float=&FNTimeOfDayScalar(Night Scale=0.5, Morning Scale=1, Evening Scale=0.5))
    )) {isSetParameters=true}
   &ScaleColor(
    Scale Alpha=&Lerp_Float(
     A=&FNTimeOfDayScalar(Night Scale=2, Morning Scale=3, Day Scale=3, Evening Scale=2)
     B=0
     Alpha=$Particles.NormalizedAge
    )
    Scale RGB=&Multiply_VectorByFloat(
     Vector=&MakeVectorFromLinearColorRGB(Color=&ColorFromCurve(ColorCurve=DataInterface() {propertyValues {
         AlphaCurve {keys []}
         BlueCurve {keys [<at 0 0.038 RCIM_Cubic>,<at 0.492 0.5>,<at 1 0.875 RCIM_Cubic>]}
         GreenCurve {keys [<at 0 0.152 RCIM_Cubic>,<at 0.492 0.625>,<at 1 0.9062 RCIM_Cubic>]}
         RedCurve {keys [<at 0 1 RCIM_Cubic>,<at 0.492 1>,<at 1 1 RCIM_Cubic>]}
        }}))
     Float=5
    )
    ScaleA=true
    ScaleRGB=true
   ) {enabled=false}
   &ScaleMeshSize(Scale Factor=&Multiply_VectorByFloat(
     Vector=&MakeVector(
      X=&FloatFromCurve(FloatCurve=DataInterface() {propertyValues {Curve {keys [<at 0 0.5 RCIM_Cubic>, {arriveTangent=0.25, interpMode=RCIM_Cubic, leaveTangent=0.25, time=0.5, value=1.5}, <at 1 0.75 RCIM_Cubic>]}}})
      Y=&FloatFromCurve(FloatCurve=DataInterface() {propertyValues {Curve {keys [<at 0.5 1 RCIM_Cubic>,<at 1 1.75 RCIM_Cubic>]}}}, Scale Curve=0.5)
      Z=&FloatFromCurve(FloatCurve=DataInterface() {propertyValues {Curve {keys [<at 0.5 1 RCIM_Cubic>,<at 1 1.75 RCIM_Cubic>]}}}, Scale Curve=0.5)
     )
     Float=70
    )) {enabled=false}
   &Color(Color=&LerpLinearColors(
     StartColor=<rgba 0.3073 0.7049 1 1>
     EndColor=<rgba 1 0.1651 0 1>
     LerpFactor=&FloatFromCurve(FloatCurve=DataInterface() {propertyValues {Curve {keys [<at 13 0 RCIM_Cubic>,<at 14 1 RCIM_Cubic>]}}}, CurveIndex=$Engine.System.Age)
    ))
   &ScaleColor(
    Color Value To Scale=$Particles.Color
    Scale Alpha=&Multiply_Float(
     A=&FloatFromCurve(
      FloatCurve=DataInterface() {propertyValues {
        bOptimizeLUT=false
        Curve {keys [
          <at 8 0.5 RCIM_Cubic>
          {arriveTangent=0.75, interpMode=RCIM_Cubic, leaveTangent=0.75, time=9.5, value=1}
          {arriveTangent=-0.0179, time=10, value=2}
          <at 11 1>
          {arriveTangent=-0.1053, interpMode=RCIM_Cubic, leaveTangent=-0.1053, time=19, value=1}
          <at 20.5 0 RCIM_Cubic>
         ]}
       }}
      CurveIndex=$Engine.System.Age
      Scale Curve=3
     )
     B=&FloatFromCurve(
      FloatCurve=DataInterface() {propertyValues {Curve {keys [{interpMode=RCIM_Cubic, tangentMode=RCTM_User, time=0, value=0}, {interpMode=RCIM_Cubic, tangentMode=RCTM_User, time=0.5, value=1}]}}}
      Scale Curve=&Multiply_Float(A=1.5, B=&FNTimeOfDayScalar(Night Scale=0.5, Morning Scale=1, Evening Scale=1))
     )
    )
    ScaleA=true
    ScaleRGB=true
   ) {module=ScaleColor001}
  ]
  rendererClasses [&NiagaraMeshRendererProperties]
  renderers [{rendererClass=&NiagaraMeshRendererProperties, rendererIndex=0}]
 }
}
summary {
 systemName=NS_BarnTowelArmory_Thrusters
 emitters [
  {emitterName=Embers, simTarget=GPUComputeSim, rendererClasses [&NiagaraSpriteRendererProperties]}
  {emitterName=ConeLight, rendererClasses [&NiagaraSpriteRendererProperties, &NiagaraLightRendererProperties]}
  {emitterName=Embers001, simTarget=GPUComputeSim, rendererClasses [&NiagaraSpriteRendererProperties]}
  {emitterName=Embers002, simTarget=GPUComputeSim, rendererClasses [&NiagaraSpriteRendererProperties]}
  {emitterName=Embers003, simTarget=GPUComputeSim, rendererClasses [&NiagaraSpriteRendererProperties]}
  {emitterName=Center_flame, rendererClasses [&NiagaraMeshRendererProperties]}
  {emitterName=Glow001, rendererClasses [&NiagaraMeshRendererProperties]}
  {emitterName=Thrusters, rendererClasses [&NiagaraMeshRendererProperties]}
 ]
 userVariables []
}
